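# Jetpack_Drop_Cue is a SoundCue for the sound of a jetpack item being dropped.
# A single Wave Player feeds Jetpack_Drop_01 into a Modulator that randomizes pitch slightly upward for variation between plays.
# It plays at reduced volume through the gadget items sound class, using the shared weapon-drop attenuation settings.
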SoundCue /Game/Athena/Sounds/Items/Jetpack/Jetpack_Drop_Cue
root=Modulator_0
nodes [
 {id=Modulator_0, children [WavePlayer_1], properties {pitchMax=1.1, pitchMin=1}}
 {id=WavePlayer_1, properties {soundWaveAssetPtr=&/Jetpack_Drop_01}}
]
properties {
 attenuationSettings=&Fort_Weapon_Drop_Att
 maxDistance=3000
 soundClassObject=&Fort_Gadgets_Items
 volumeMultiplier=0.3
}
