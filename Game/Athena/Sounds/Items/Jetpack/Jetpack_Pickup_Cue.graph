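# Sound cue for picking up the Jetpack item in Athena, classed under gadget item sounds.
# A single wave player for the pickup sound feeds a Modulator that randomly raises pitch slightly for variation.
# It plays at reduced volume with the shared weapon pickup attenuation.
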SoundCue /Game/Athena/Sounds/Items/Jetpack/Jetpack_Pickup_Cue
root=Modulator_0
nodes [
 {id=Modulator_0, children [WavePlayer_0], properties {pitchMax=1.1, pitchMin=1}}
 {id=WavePlayer_0, properties {soundWaveAssetPtr=&/Jetpack_Pickup_01}}
]
properties {
 attenuationSettings=&Fort_Weapon_Pickup_Att
 maxDistance=3500
 soundClassObject=&Fort_Gadgets_Items
 volumeMultiplier=0.3
}
